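# Niagara swing effect for the ProfessorPup male one-handed pickaxe, combining ribbon trails, spray sprites, flickering lights and gold offcast sparkles.
# Four ribbons spawn by distance moved, animating material parameters over emitter age, while lights and spray follow the swing_FX socket.
# User.Variant switches looks: orange-red with dark trails (0), gold sparkles with a white ribbon (1), or default purple (2).
# SnowTrail001's RGB-scale switch is fixed at index 0, ignoring the variant.
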NiagaraSystem /BRCosmetics/Weapons/FORT_Melee/Pickaxe_ID_412_ProfessorPupMale1h/FX/NS_Pickaxe_ProfessorPup_Male_Pickaxe1h_Swing
dependencies {
 usedDataInterfaces [
  &NiagaraDataInterfaceCurve
  &NiagaraDataInterfaceSkeletalMesh
  &NiagaraDataInterfaceSpriteRendererInfo
  &NiagaraDataInterfaceVectorCurve
 ]
 usedDynamicInputs []
 usedModules [&SystemState]
}
emitters {
 Left_Light {
  emitterUpdateScript [
   &EmitterState(Inactive Response=::NewEnumerator1, Life Cycle Mode=::NewEnumerator1, Loop Behavior=::NewEnumerator1, Loop Duration=0.5, UseLoopDelay=true)
   &SpawnBurst_Instantaneous(Spawn Count=2, Use Spawn Probability=false, UseLoopCountLimit=false)
  ]
  particleSpawnScript [
   &InitializeParticle(Sprite Size=<5 5>, Sprite Size Mode=::NewEnumerator4, UsePositionOffset=false, Write Material Random=true)
   &Color(Color=&MergeHelper_ColorSwitch(
     0=&MakeLinearColorFromVectorAndFloat("Vector (RGB)"=<10 1.9542 0>, "Float (Alpha)"=0.5)
     1=<rgba 0 0 0 0>
     2=&MakeLinearColorFromVectorAndFloat("Vector (RGB)"=<2.9355 0 6>, "Float (Alpha)"=0.5)
     3=?
     Index=&MakeIntFromFloat(Float=$User.Variant)
     Num=::NewEnumerator1
    ))
  ]
  particleUpdateScript [
   /SetVariables(Particles.LightExponent=0.5) {isSetParameters=true}
   &ParticleState()
   &CascadeConversion_LightProperties(LightBrightnessScale=8, LightColorScale=<1 1 1>, LightRadiusScale=0.7)
   &ScaleColor(
    Scale Alpha=&FloatFromCurve(FloatCurve=DataInterface() {propertyValues {Curve {keys [
         {arriveTangent=1.5, time=0, value=0}
         {arriveTangent=-1.5, leaveTangent=-1.5, time=0.1, value=1.5}
         {arriveTangent=1.25, leaveTangent=1.25, time=0.175, value=0}
         {arriveTangent=-1.25, leaveTangent=-1.25, time=0.25, value=1.25}
         <at 0.325 0>
         {arriveTangent=3.6364, interpMode=RCIM_Cubic, leaveTangent=3.6364, time=0.525, value=0}
         <at 0.6 1>
         <at 0.675 0>
        ]}}})
    ScaleA=true
    ScaleRGB=true
   )
   &SkeletalMeshLocation(
    Bone / Socket Index=/CascadeToNiagaraConverter/NiagaraScripts/DynamicInputScripts/SequentialInt(Use Array Data:NiagaraBool=false, First Number:NiagaraInt32=0, Sequence:NiagaraDataInterfaceArrayInt32=?, Last Number:NiagaraInt32=0)
    MaskSpawnGroup=false
    Mesh Sampling Type=::NewEnumerator5
    Orientation Sampling=::NewEnumerator1
    Skeletal Mesh=DataInterface() {propertyValues {
      DefaultMesh=null
      ExcludeBoneName=root
      OptionalComponentClass=null
      PreviewMesh=/BRCosmetics/Weapons/FORT_Melee/Pickaxe_Professor_Pup_Male_1h/Meshes/Professor_Pup_Male_1h_Axe
      SkinningMode=SkinOnTheFly
      SoftSourceActor=null
      SourceMode=Default
      UvSetIndex=0
      WholeMeshLOD=-1
      bExcludeBone=false
      bReadDeformedGeometry=true
      bRequireCurrentFrameData=true
      ComponentTags []
      FilteredBones []
      FilteredSockets [swing_FX]
      MeshUserParameter {parameter {name=None, typeDefHandle {registeredTypeIndex=1}, varData []}}
      SamplingRegions []
     }}
    Socket Sampling Mode=::NewEnumerator5
   )
  ]
  rendererClasses [&NiagaraLightRendererProperties]
  renderers [{rendererClass=&NiagaraLightRendererProperties, rendererIndex=0}]
 }
 NoseTrail {
  emitterUpdateScript [
   &EmitterState(Life Cycle Mode=::NewEnumerator1, Loop Behavior=::NewEnumerator1, Loop Duration=0.3, UseLoopDelay=true)
   &SpawnPerUnit(
    Module.Use Max Movement Threshold=true
    Module.Use Movement Tolerance=true
    Movement Tolerance=0.1
    Spawn Spacing:NiagaraFloat=&Divide_Float(A=1, B=&Divide_Float(A=1, B=10))
    Use Spawn Probability=false
   )
  ]
  particleSpawnScript [
   &InitializeParticle(
    Lifetime=0.2
    Ribbon Width=&Multiply_Float(A=10, B=2)
    Ribbon Width Mode=::NewEnumerator1
    UsePositionOffset=false
    Write Material Random=true
   )
   &Color(Color=&MergeHelper_ColorSwitch(
     0=&MakeLinearColorFromVectorAndFloat("Vector (RGB)"=<1 0.126 0>, "Float (Alpha)"=0.7)
     1=<rgba 0 0 0 0>
     2=&MakeLinearColorFromVectorAndFloat("Vector (RGB)"=<0.4893 0 1>, "Float (Alpha)"=0.7)
     3=?
     Index=&MakeIntFromFloat(Float=$User.Variant)
     Num=::NewEnumerator1
    ))
  ]
  particleUpdateScript [
   /SetVariables(Particles.ScaledRibbonWidth:NiagaraFloat=&Multiply_Float(
     A=$Particles.RibbonWidth
     B=&FloatFromCurve(FloatCurve=DataInterface() {propertyValues {Curve {keys [
          {arriveTangent=-0.9322, interpMode=RCIM_Cubic, leaveTangent=-0.9322, tangentMode=RCTM_User, time=0, value=0.8994}
          <at 0.5976 0.1 RCIM_Cubic>
         ]}}})
    )) {isSetParameters=true}
   &ParticleState()
   &ScaleColor(
    Scale Alpha=&FloatFromCurve(FloatCurve=DataInterface() {propertyValues {Curve {keys [<at 0.0106 0 RCIM_Cubic>,<at 0.0995 0.74>,<at 1 0>]}}})
    Scale RGB=&VectorFromCurve(VectorCurve=DataInterface() {propertyValues {
       XCurve {keys [<at 0 0.5>,<at 0.2 3>,<at 0.4 0.5>]}
       YCurve {keys [<at 0 0.5>,<at 0.2 3>,<at 0.4 0.5>]}
       ZCurve {keys [<at 0 0.5>,<at 0.2 3>,<at 0.4 0.5>]}
      }})
    ScaleA=true
    ScaleRGB=true
   )
   &DynamicMaterialParameters(
    Index 0 Param 1=&FloatFromCurve(
     FloatCurve=DataInterface() {propertyValues {Curve {keys [
         {arriveTangent=-23.6538, interpMode=RCIM_Cubic, leaveTangent=-23.6538, tangentMode=RCTM_User, time=0, value=4.0503}
         {arriveTangent=-4.507, interpMode=RCIM_Cubic, leaveTangent=-4.507, time=0.2499, value=1.1157}
         <at 0.8989 -0.0009 RCIM_Cubic>
        ]}}}
     CurveIndex=$Emitter.LoopedAge
    )
    Index 0 Param 2=1
    Index 0 Param 3=&FloatFromCurve(
     FloatCurve=DataInterface() {propertyValues {Curve {keys [
         {arriveTangent=-3.043, interpMode=RCIM_Cubic, leaveTangent=-3.043, tangentMode=RCTM_User, time=0.199, value=0.1267}
         <at 0.9991 -0.4903 RCIM_Cubic>
        ]}}}
     CurveIndex=$Emitter.LoopedAge
    )
    Index 0 Param 4=&FloatFromCurve(
     FloatCurve=DataInterface() {propertyValues {Curve {keys [
         <at 0.0986 0 RCIM_Cubic>
         {arriveTangent=0.4869, interpMode=RCIM_Cubic, leaveTangent=0.4869, time=0.1428, value=0.4623}
         {arriveTangent=-0.5406, interpMode=RCIM_Cubic, leaveTangent=-0.5406, time=0.4999, value=0.1954}
         <at 0.9981 0 RCIM_Cubic>
        ]}}}
     CurveIndex=$Emitter.LoopedAge
    )
    Module.Param0WriteEnabled=true
    Module.Param1WriteEnabled=true
    Module.Param2WriteEnabled=true
    Module.Param3WriteEnabled=true
   )
   &CameraOffset(Camera Offset Amount=20)
  ]
  rendererClasses [&NiagaraRibbonRendererProperties]
  renderers [{rendererClass=&NiagaraRibbonRendererProperties}]
 }
 Offcast_Gold {
  emitterUpdateScript [
   &EmitterState(Inactive Response=::NewEnumerator1, Life Cycle Mode=::NewEnumerator1, Loop Behavior=::NewEnumerator1, Loop Duration=0.1, UseLoopDelay=true)
   &SpawnRate(
    SpawnRate=&MergeHelper_FloatSwitch(0=0, 1=40, 2=0, 3=?, Index=&MakeIntFromFloat(Float=$User.Variant), Num=::NewEnumerator1)
    Use Spawn Probability=false
   )
  ]
  particleSpawnScript [
   /SetVariables(Particles.SpriteRotationRate=&RandomRangeFloat(Minimum=-0.125, Maximum=0.25, Evaluation Type=::NewEnumerator1)) {isSetParameters=true}
   &InitializeParticle(
    Lifetime=&RandomRangeFloat(Minimum=0.5, Maximum=0.8, Evaluation Type=::NewEnumerator1)
    Sprite Rotation Angle=&RandomRangeFloat(Minimum=0, Maximum=1, Evaluation Type=::NewEnumerator1)
    Sprite Rotation Mode=::NewEnumerator1
    Sprite Size=&Vector2DFromFloat(Value=&RandomRangeFloat(Minimum=1, Maximum=5, Evaluation Type=::NewEnumerator1))
    Sprite Size Mode=::NewEnumerator4
    UsePositionOffset=false
    Write Material Random=true
   )
   &Color(Color=&MakeLinearColorFromVectorAndFloat(
     "Vector (RGB)"=&RandomRangeVector(Minimum=<2 1.4074 0.11>, Maximum=<20 14.0736 1.1>, Evaluation Type=::NewEnumerator1)
     "Float (Alpha)"=1
    ))
   &ShapeLocation(MaskSpawnGroup=false, Override Seed=false, Sphere Radius=10, UseOwnerScale=true)
   &InheritVelocity(Inherited Velocity Amount Scale=<0.01 0.01 0.01>, Inherited Velocity Speed Limit=0, Speed Limit=false, UseSpeedThreshold=false)
  ]
  particleUpdateScript [
   &ParticleState()
   &SpriteRotationRate(Rotation Rate=&AngleConversion(
     Angle:NiagaraFloat=$Particles.SpriteRotationRate
     Angle Input:ENiagara_AngleInput=::NewEnumerator1
     Angle Output:ENiagara_AngleInput=::NewEnumerator0
    ))
   &AccelerationForce(Acceleration=<0 0 -25>)
   &SolveForcesAndVelocity(Clamp Velocity=false, Limit Acceleration=false)
   &ScaleColor(
    Scale Alpha=&FloatFromCurve(FloatCurve=DataInterface() {propertyValues {Curve {keys [<at 0 0>,<at 0.5 1>,<at 1 0>]}}})
    ScaleA=true
    ScaleRGB=true
   )
  ]
  rendererClasses [&NiagaraSpriteRendererProperties]
  renderers [{rendererClass=&NiagaraSpriteRendererProperties}]
 }
 SnowTrail {
  emitterUpdateScript [
   &EmitterState(Life Cycle Mode=::NewEnumerator1, Loop Behavior=::NewEnumerator1, Loop Duration=0.3, UseLoopDelay=true)
   &SpawnPerUnit(
    Module.Use Max Movement Threshold=true
    Module.Use Movement Tolerance=true
    Movement Tolerance=0.1
    Spawn Spacing:NiagaraFloat=&Divide_Float(A=1, B=&Divide_Float(A=1, B=60))
    Use Spawn Probability=false
   )
  ]
  particleSpawnScript [
   &InitializeParticle(
    Lifetime=0.45
    Ribbon Width=&Multiply_Float(A=7, B=2)
    Ribbon Width Mode=::NewEnumerator1
    UsePositionOffset=false
    Write Material Random=true
   )
   &Color(Color=&MergeHelper_ColorSwitch(
     0=&MakeLinearColorFromVectorAndFloat("Vector (RGB)"=<0.0885 0.0885 0.0885>, "Float (Alpha)"=1)
     1=<rgba 0 0 0 0>
     2=&MakeLinearColorFromVectorAndFloat("Vector (RGB)"=<0.9785 0 2>, "Float (Alpha)"=1)
     3=?
     Index=&MakeIntFromFloat(Float=$User.Variant)
     Num=::NewEnumerator1
    ))
  ]
  particleUpdateScript [
   &ParticleState()
   &ScaleColor(
    Scale Alpha=&FloatFromCurve(FloatCurve=DataInterface() {propertyValues {Curve {keys [<at 0.0106 0 RCIM_Cubic>,<at 0.0995 1 RCIM_Cubic>,<at 0.8718 0 RCIM_Cubic>]}}})
    Scale RGB=&MergeHelper_VectorSwitch(
     0=&VectorFromCurve(VectorCurve=DataInterface() {propertyValues {
        XCurve {keys [<at 0 0.0885 RCIM_Cubic>,<at 0.2 3 RCIM_Cubic>,<at 0.4 0.0885 RCIM_Cubic>]}
        YCurve {keys [<at 0 0.0885 RCIM_Cubic>,<at 0.2 3 RCIM_Cubic>,<at 0.4 0.0885 RCIM_Cubic>]}
        ZCurve {keys [<at 0 0.0885 RCIM_Cubic>,<at 0.2 3 RCIM_Cubic>,<at 0.4 0.0885 RCIM_Cubic>]}
       }})
     1=<0 0 0>
     2=&VectorFromCurve(VectorCurve=DataInterface() {propertyValues {
        XCurve {keys [<at 0 0.0885 RCIM_Cubic>,<at 0.2 3 RCIM_Cubic>,<at 0.4 0.0885 RCIM_Cubic>]}
        YCurve {keys [<at 0 0.0885 RCIM_Cubic>,<at 0.2 3 RCIM_Cubic>,<at 0.4 0.0885 RCIM_Cubic>]}
        ZCurve {keys [<at 0 0.0885 RCIM_Cubic>,<at 0.2 3 RCIM_Cubic>,<at 0.4 0.0885 RCIM_Cubic>]}
       }})
     3=?
     Index=&MakeIntFromFloat(Float=$User.Variant)
     Num=::NewEnumerator1
    )
    ScaleA=true
    ScaleRGB=true
   )
   &DynamicMaterialParameters(
    Index 0 Param 1=&FloatFromCurve(
     FloatCurve=DataInterface() {propertyValues {Curve {keys [
         {arriveTangent=-23.6538, interpMode=RCIM_Cubic, leaveTangent=-23.6538, tangentMode=RCTM_User, time=0, value=4.0503}
         {arriveTangent=-4.507, interpMode=RCIM_Cubic, leaveTangent=-4.507, time=0.2499, value=1.1157}
         <at 0.8989 -0.0009 RCIM_Cubic>
        ]}}}
     CurveIndex=$Emitter.LoopedAge
    )
    Index 0 Param 2=1
    Index 0 Param 3=1
    Index 0 Param 4=&FloatFromCurve(
     FloatCurve=DataInterface() {propertyValues {Curve {keys [
         <at 0.0986 0 RCIM_Cubic>
         {arriveTangent=0.4869, interpMode=RCIM_Cubic, leaveTangent=0.4869, time=0.1428, value=0.4623}
         {arriveTangent=-0.5406, interpMode=RCIM_Cubic, leaveTangent=-0.5406, time=0.4999, value=0.1954}
         <at 0.9981 0 RCIM_Cubic>
        ]}}}
     CurveIndex=$Emitter.LoopedAge
    )
    Module.Param0WriteEnabled=true
    Module.Param1WriteEnabled=true
    Module.Param2WriteEnabled=true
    Module.Param3WriteEnabled=true
   )
  ]
  rendererClasses [&NiagaraRibbonRendererProperties]
  renderers [{rendererClass=&NiagaraRibbonRendererProperties}]
 }
 SnowTrail001 {
  emitterUpdateScript [
   &EmitterState(Life Cycle Mode=::NewEnumerator1, Loop Behavior=::NewEnumerator1, Loop Duration=0.3, UseLoopDelay=true)
   &SpawnPerUnit(
    Module.Use Max Movement Threshold=true
    Module.Use Movement Tolerance=true
    Movement Tolerance=0.1
    Spawn Spacing:NiagaraFloat=&Divide_Float(A=1, B=&Divide_Float(A=1, B=60))
    Use Spawn Probability=false
   )
  ]
  particleSpawnScript [
   &InitializeParticle(
    Lifetime=0.45
    Ribbon Width=&Multiply_Float(A=15, B=2)
    Ribbon Width Mode=::NewEnumerator1
    UsePositionOffset=false
    Write Material Random=true
   )
   &Color(Color=&MergeHelper_ColorSwitch(
     0=&MakeLinearColorFromVectorAndFloat("Vector (RGB)"=<0.0885 0.0885 0.0885>, "Float (Alpha)"=1)
     1=<rgba 0 0 0 0>
     2=&MakeLinearColorFromVectorAndFloat("Vector (RGB)"=<0.4893 0 1>, "Float (Alpha)"=1)
     3=?
     Index=&MakeIntFromFloat(Float=$User.Variant)
     Num=::NewEnumerator1
    ))
  ]
  particleUpdateScript [
   &ParticleState()
   &ScaleColor(
    Scale Alpha=&FloatFromCurve(FloatCurve=DataInterface() {propertyValues {Curve {keys [<at 0.0106 0 RCIM_Cubic>,<at 0.0995 1 RCIM_Cubic>,<at 0.8718 0 RCIM_Cubic>]}}})
    Scale RGB=&MergeHelper_VectorSwitch(
     0=&VectorFromCurve(VectorCurve=DataInterface() {propertyValues {
        XCurve {keys [<at 0 0.0885 RCIM_Cubic>,<at 0.2 3 RCIM_Cubic>,<at 0.4 0.0885 RCIM_Cubic>]}
        YCurve {keys [<at 0 0.0885 RCIM_Cubic>,<at 0.2 3 RCIM_Cubic>,<at 0.4 0.0885 RCIM_Cubic>]}
        ZCurve {keys [<at 0 0.0885 RCIM_Cubic>,<at 0.2 3 RCIM_Cubic>,<at 0.4 0.0885 RCIM_Cubic>]}
       }})
     1=<0 0 0>
     2=&VectorFromCurve(VectorCurve=DataInterface() {propertyValues {
        XCurve {keys [
          <at 0 0.6502 RCIM_Cubic>
          {arriveTangent=-1.4245, interpMode=RCIM_Cubic, leaveTangent=-1.4245, time=0.2, value=0.4893}
          <at 0.4 0.0804 RCIM_Cubic>
         ]}
        YCurve {keys [
          <at 0 0.5643 RCIM_Cubic>
          {arriveTangent=-1.2362, interpMode=RCIM_Cubic, leaveTangent=-1.2362, time=0.2, value=0}
          <at 0.4 0.0698 RCIM_Cubic>
         ]}
        ZCurve {keys [<at 0 0.8 RCIM_Cubic>, {arriveTangent=-1.7526, interpMode=RCIM_Cubic, leaveTangent=-1.7526, time=0.2, value=1}, <at 0.4 0.099 RCIM_Cubic>]}
       }})
     3=?
     Index=0
     Num=::NewEnumerator1
    )
    ScaleA=true
    ScaleRGB=true
   )
   &DynamicMaterialParameters(
    Index 0 Param 1=&FloatFromCurve(
     FloatCurve=DataInterface() {propertyValues {Curve {keys [
         {arriveTangent=-23.6538, interpMode=RCIM_Cubic, leaveTangent=-23.6538, tangentMode=RCTM_User, time=0, value=4.0503}
         {arriveTangent=-4.507, interpMode=RCIM_Cubic, leaveTangent=-4.507, time=0.2499, value=1.1157}
         <at 0.8989 -0.0009 RCIM_Cubic>
        ]}}}
     CurveIndex=$Emitter.LoopedAge
    )
    Index 0 Param 2=1
    Index 0 Param 3=1
    Index 0 Param 4=&FloatFromCurve(
     FloatCurve=DataInterface() {propertyValues {Curve {keys [
         <at 0.0986 0 RCIM_Cubic>
         {arriveTangent=0.4869, interpMode=RCIM_Cubic, leaveTangent=0.4869, time=0.1428, value=0.4623}
         {arriveTangent=-0.5406, interpMode=RCIM_Cubic, leaveTangent=-0.5406, time=0.4999, value=0.1954}
         <at 0.9981 0 RCIM_Cubic>
        ]}}}
     CurveIndex=$Emitter.LoopedAge
    )
    Module.Param0WriteEnabled=true
    Module.Param1WriteEnabled=true
    Module.Param2WriteEnabled=true
    Module.Param3WriteEnabled=true
   )
  ]
  rendererClasses [&NiagaraRibbonRendererProperties]
  renderers [{rendererClass=&NiagaraRibbonRendererProperties}]
 }
 SnowTrail002 {
  emitterUpdateScript [
   &EmitterState(Life Cycle Mode=::NewEnumerator1, Loop Behavior=::NewEnumerator1, Loop Duration=0.3, UseLoopDelay=true)
   &SpawnPerUnit(
    Module.Use Max Movement Threshold=true
    Module.Use Movement Tolerance=true
    Movement Tolerance=0.1
    Spawn Spacing:NiagaraFloat=&Divide_Float(A=1, B=&Divide_Float(A=1, B=5))
    Use Spawn Probability=false
   )
  ]
  particleSpawnScript [
   &InitializeParticle(
    Lifetime=&MergeHelper_FloatSwitch(0=0, 1=0.15, 2=0, 3=?, Index=&MakeIntFromFloat(Float=$User.Variant), Num=::NewEnumerator1)
    Ribbon Width=&Multiply_Float(A=15, B=2)
    Ribbon Width Mode=::NewEnumerator1
    UsePositionOffset=false
    Write Material Random=true
   )
   &Color(Color=&MakeLinearColorFromVectorAndFloat("Vector (RGB)"=<1 1 1>, "Float (Alpha)"=0.5))
  ]
  particleUpdateScript [
   &ParticleState()
   &ScaleColor(
    Scale Alpha=&FloatFromCurve(FloatCurve=DataInterface() {propertyValues {Curve {keys [<at 0.0106 0 RCIM_Cubic>,<at 0.2 1 RCIM_Cubic>,<at 1 0 RCIM_Cubic>]}}})
    Scale RGB=&VectorFromCurve(VectorCurve=DataInterface() {propertyValues {
       XCurve {keys [<at 0 0.0885 RCIM_Cubic>,<at 0.2 3 RCIM_Cubic>,<at 0.4 0.0885 RCIM_Cubic>]}
       YCurve {keys [<at 0 0.0885 RCIM_Cubic>,<at 0.2 3 RCIM_Cubic>,<at 0.4 0.0885 RCIM_Cubic>]}
       ZCurve {keys [<at 0 0.0885 RCIM_Cubic>,<at 0.2 3 RCIM_Cubic>,<at 0.4 0.0885 RCIM_Cubic>]}
      }})
    ScaleA=true
    ScaleRGB=true
   )
   &DynamicMaterialParameters(
    Index 0 Param 1=&FloatFromCurve(FloatCurve=DataInterface() {propertyValues {Curve {keys [<at 0 1>,<at 1 1>]}}}, CurveIndex=$Emitter.LoopedAge)
    Index 0 Param 3=1
    Index 0 Param 4=&FloatFromCurve(FloatCurve=DataInterface() {propertyValues {Curve {keys [<at 0 0>,<at 0.5 1>,<at 1 0>]}}}, CurveIndex=$Emitter.LoopedAge)
    Module.Param0WriteEnabled=true
    Module.Param1WriteEnabled=true
    Module.Param2WriteEnabled=true
    Module.Param3WriteEnabled=true
   )
  ]
  rendererClasses [&NiagaraRibbonRendererProperties]
  renderers [{rendererClass=&NiagaraRibbonRendererProperties}]
 }
 Spray {
  emitterUpdateScript [
   &EmitterState(Life Cycle Mode=::NewEnumerator1, Loop Behavior=::NewEnumerator1, Loop Duration=0.08, UseLoopDelay=true)
   &SpawnRate(SpawnRate=70, Use Spawn Probability=false)
  ]
  particleSpawnScript [
   /SetVariables(Particles.SpriteRotationRate=&RandomRangeFloat(Minimum=-0.5, Maximum=0.5, Evaluation Type=::NewEnumerator1)) {isSetParameters=true}
   &InitializeParticle(
    Lifetime=0.3
    Sprite Rotation Angle=&RandomRangeFloat(Minimum=0, Maximum=1, Evaluation Type=::NewEnumerator1)
    Sprite Rotation Mode=::NewEnumerator1
    Sprite Size=&Vector2DFromFloat(Value=&RandomRangeFloat(Minimum=7, Maximum=20, Evaluation Type=::NewEnumerator1))
    Sprite Size Mode=::NewEnumerator4
    UsePositionOffset=false
    Write Material Random=true
   )
   &Color(Color=&MergeHelper_ColorSwitch(
     0=&MakeLinearColorFromVectorAndFloat("Vector (RGB)"=<1 0.0104 0.0104>, "Float (Alpha)"=1)
     1=<rgba 0 0 0 0>
     2=&MakeLinearColorFromVectorAndFloat("Vector (RGB)"=<0.4893 0 1>, "Float (Alpha)"=1)
     3=?
     Index=&MakeIntFromFloat(Float=$User.Variant)
     Num=::NewEnumerator1
    ))
   &AddVelocity(
    Cone Angle=0
    Cone Angle Mode:ENiagara_AngleInput=::NewEnumerator1
    Module.UseConeFalloff=false
    Module.UseVelDistribution=false
    Override Seed=false
    Velocity Mode=::NewEnumerator2
    Velocity Speed=&RandomRangeFloat(Minimum=5, Maximum=10, Evaluation Type=::NewEnumerator1)
   )
   &ShapeLocation(MaskSpawnGroup=false, Override Seed=false, Sphere Radius=2.5, UseOwnerScale=true)
  ]
  particleUpdateScript [
   &ParticleState()
   &SkeletalMeshLocation(
    Bone / Socket Index=&SequentialInt(Use Array Data:NiagaraBool=false, First Number:NiagaraInt32=0, Sequence:NiagaraDataInterfaceArrayInt32=?, Last Number:NiagaraInt32=0)
    MaskSpawnGroup=false
    Mesh Sampling Type=::NewEnumerator5
    Orientation Sampling=::NewEnumerator1
    Skeletal Mesh=DataInterface() {propertyValues {
      DefaultMesh=null
      ExcludeBoneName=root
      OptionalComponentClass=null
      PreviewMesh=&Professor_Pup_Male_1h_Axe
      SkinningMode=SkinOnTheFly
      SoftSourceActor=null
      SourceMode=Default
      UvSetIndex=0
      WholeMeshLOD=-1
      bExcludeBone=false
      bReadDeformedGeometry=true
      bRequireCurrentFrameData=true
      ComponentTags []
      FilteredBones []
      FilteredSockets [swing_FX]
      MeshUserParameter {parameter {name=None, typeDefHandle {registeredTypeIndex=1}, varData []}}
      SamplingRegions []
     }}
    Socket Sampling Mode=::NewEnumerator5
   )
   &SpriteRotationRate(Rotation Rate=&AngleConversion(
     Angle:NiagaraFloat=$Particles.SpriteRotationRate
     Angle Input:ENiagara_AngleInput=::NewEnumerator1
     Angle Output:ENiagara_AngleInput=::NewEnumerator0
    ))
   &ScaleSpriteSize(
    Scale Sprite Size Mode=::NewEnumerator0
    Uniform Scale Factor=&FloatFromCurve(FloatCurve=DataInterface() {propertyValues {Curve {keys [
         {arriveTangent=0.729, time=0.0008, value=1}
         {arriveTangent=-0.0747, interpMode=RCIM_Cubic, leaveTangent=-0.0747, time=0.2244, value=0.7305}
         {arriveTangent=-0.9456, interpMode=RCIM_Cubic, leaveTangent=-0.9456, time=0.6659, value=0.9503}
         <at 1 -0.0029>
        ]}}})
   )
   &ScaleColor(
    Scale Alpha=&FloatFromCurve(FloatCurve=DataInterface() {propertyValues {Curve {keys [{arriveTangent=-0.1781, time=0, value=1}, {arriveTangent=-0.8219, leaveTangent=-0.8219, time=0.6015, value=0.8219}, <at 1 0>]}}})
    Scale RGB=<5 5 5>
    ScaleA=true
    ScaleRGB=true
   )
   &Drag(Drag=2, Use Linear Drag=true, Use Rotational Drag=true)
   &SolveForcesAndVelocity(Clamp Velocity=false, Limit Acceleration=false)
   &SubUVAnimation(
    End Frame Range Override=35
    "Sprite Renderer "=DataInterface() {propertyValues {SpriteRenderer=&self:Spray_0.NiagaraSpriteRendererProperties_10}}
    UseEndFrame=false
    UseStartFrame=false
   )
  ]
  rendererClasses [&NiagaraSpriteRendererProperties]
  renderers [{rendererClass=&NiagaraSpriteRendererProperties}]
 }
}
summary {
 systemName=NS_Pickaxe_ProfessorPup_Male_Pickaxe1h_Swing
 emitters [
  {emitterName=Left_Light, rendererClasses [&NiagaraLightRendererProperties]}
  {emitterName=Spray, rendererClasses [&NiagaraSpriteRendererProperties]}
  {emitterName=SnowTrail, rendererClasses [&NiagaraRibbonRendererProperties]}
  {emitterName=SnowTrail001, rendererClasses [&NiagaraRibbonRendererProperties]}
  {emitterName=NoseTrail, rendererClasses [&NiagaraRibbonRendererProperties]}
  {emitterName=SnowTrail002, rendererClasses [&NiagaraRibbonRendererProperties]}
  {emitterName=Offcast_Gold, rendererClasses [&NiagaraSpriteRendererProperties]}
 ]
 userVariables [User.Variant {type=NiagaraFloat, defaultValue=2, description=""}]
}
